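# Looping ambient sound cue for the Gem Fragment quest actors in Battle Pass S19.
# A single looping wave player feeds the GemFragment_Amb_Loop_01 wave straight to the output, with a boosted volume multiplier.
# It routes through the Fort_Quests sound class and references a dedicated Gem Fragment loop attenuation asset for distance falloff.
# The cue also stores attenuation override values, though the override switch is not shown as enabled.
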SoundCue /BattlePassS19/QuestActors/GemFragments/Assets/SFX/GemFragment_AmbientLoop_Cue
root=WavePlayer_2
nodes [{id=WavePlayer_2, properties {bLooping=true, soundWaveAssetPtr=&/GemFragment_Amb_Loop_01}}]
properties {
 attenuationSettings=&/Fort_GemFragment_Ambient_Loop_Att
 maxDistance=1300
 soundClassObject=&Fort_Quests
 volumeMultiplier=2
 attenuationOverrides {falloffDistance=700, attenuationShapeExtents {x=100}}
}
