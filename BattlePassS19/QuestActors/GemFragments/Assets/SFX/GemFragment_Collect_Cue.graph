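# GemFragment_Collect_Cue is a Sound Cue for collecting a gem fragment among the BattlePassS19 quest actors.
# A single wave player feeds the root directly with GemFragment_Collect_01, with no randomization or modulation.
# It routes through the Fort_UI sound class and overrides concurrency to one instance, stopping the oldest so repeated pickups do not stack.
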SoundCue /BattlePassS19/QuestActors/GemFragments/Assets/SFX/GemFragment_Collect_Cue
root=WavePlayer_0
nodes [{id=WavePlayer_0, properties {soundWaveAssetPtr=&/GemFragment_Collect_01}}]
properties {
 bOverrideConcurrency=true
 maxDistance=2097152
 soundClassObject=&Fort_UI
 volumeMultiplier=1
 concurrencyOverrides {maxCount=1, resolutionRule=StopOldest}
}
